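# Animation Blueprint for the DistantEchoCastle female character's hood accessory, animating left and right scarf bones over the pose copied from the parent mesh.
# Weighted constraints to head, spine and clavicles feed additive rotations, then spring-based Anim Dynamics on four scarf bones.
# Each update, pistol (outside emotes), assault rifle and shotgun types enable an extra right-scarf rotation.
# Aiming with those weapons also lowers the right-side dynamics alpha.
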AnimBlueprint /Game/Characters/Player/Female/Medium/Bodies/F_MED_DistantEchoCastle/Meshes/Parts/F_MED_DistantEchoCastle_hood_FaceAcc_AnimBP
target_skeleton=&/F_MED_DistantEchoCastle_hood_FaceAcc_Skeleton
graphs [
 AnimGraph {nodes [
   n1 = "Constraint - dyn_scarf_l (PPPP)"!(
    ComponentPose=Local To Component(LocalPose=Copy Pose From Mesh() {bUseAttachedParent=true})
    Alpha=1
    ConstraintWeights_0=0.5504
    ConstraintWeights_1:float=0.2862
    ConstraintWeights_2:float=0.0883
    ConstraintWeights_3:float=0.0751
   ) -> (Pose) {
    alphaBoolBlend {blendOption=Linear}
    boneToModify {boneName=dyn_scarf_l}
    constraintSetup [{targetBone {boneName=head}}, {targetBone {boneName=spine_05}}, {targetBone {boneName=clavicle_l}}, {targetBone {boneName=dyn_scarf_l}}]
    constraintWeights [1,1,1,1]
   }
   n2 = "Constraint - dyn_scarf_r (PPPP)"!(
    ComponentPose=@n1
    Alpha=1
    ConstraintWeights_0=0.5461
    ConstraintWeights_1:float=0.294
    ConstraintWeights_2:float=0.0901
    ConstraintWeights_3:float=0.0698
   ) -> (Pose) {
    alphaBoolBlend {blendOption=Linear}
    boneToModify {boneName=dyn_scarf_r}
    constraintSetup [{targetBone {boneName=head}}, {targetBone {boneName=spine_05}}, {targetBone {boneName=clavicle_r}}, {targetBone {boneName=dyn_scarf_r}}]
    constraintWeights [1,1,1,1]
   }
   n3 = "Transform (Modify) Bone - Bone: dyn_scarf_r"(ComponentPose=@n2, Alpha=Get offset_weapon() -> (offset_weapon:double), Translation="0,0,0", Rotation="-15,-20,15", Scale="1,1,1") {
    rotationMode=BMM_Additive
    rotationSpace=BCS_BoneSpace
    scaleMode=BMM_Additive
    scaleSpace=BCS_BoneSpace
    translationMode=BMM_Additive
    translationSpace=BCS_BoneSpace
    alphaBoolBlend {blendOption=Linear}
    boneToModify {boneName=dyn_scarf_r}
    rotation=<rot -15 -20 15>
   }
   n4 = "Transform (Modify) Bone - Bone: dyn_scarf_r"(ComponentPose=@n3, Alpha=1, Translation="0,0,0", Rotation="0,15,0", Scale="1,1,1") {
    rotationMode=BMM_Additive
    rotationSpace=BCS_BoneSpace
    scaleMode=BMM_Additive
    scaleSpace=BCS_BoneSpace
    translationMode=BMM_Additive
    translationSpace=BCS_BoneSpace
    alphaBoolBlend {blendOption=Linear}
    boneToModify {boneName=dyn_scarf_r}
    rotation {yaw=15}
   }
   n5 = "Transform (Modify) Bone - Bone: dyn_scarf_l"(ComponentPose=@n4, Alpha=1, Translation="0,0,0", Rotation="0,15,0", Scale="1,1,1") {
    rotationMode=BMM_Additive
    rotationSpace=BCS_BoneSpace
    scaleMode=BMM_Additive
    scaleSpace=BCS_BoneSpace
    translationMode=BMM_Additive
    translationSpace=BCS_BoneSpace
    alphaBoolBlend {blendOption=Linear}
    boneToModify {boneName=dyn_scarf_l}
    rotation {yaw=15}
   }
   n6 = Anim Dynamics - Bone: dyn_scarf_l(ComponentPose=@n5, Alpha=0.4, ExternalForce="0,0,0") {
    angularSpringConstant=800
    bAngularSpring=true
    bGravityOverrideInSimSpace=true
    bOverrideAngularDamping=true
    bUseGravityOverride=true
    gravityScale=0.3
    simulationSpace=BoneRelative
    alphaBoolBlend {blendOption=Linear}
    boundBone {boneName=dyn_scarf_l}
    componentAppliedLinearAccClamp=<400 400 400>
    componentLinearAccScale=<0.1 0.1 0.1>
    componentLinearVelScale=<0.1 0.1 0.1>
    gravityOverride {z=-980}
    physicsBodyDefinitions [{
      sphereCollisionRadius=0
      boundBone {boneName=dyn_scarf_l}
      boxExtents=<45 45 45>
      constraintSetup {twistAxis=AxisZ, angularLimitsMax=<0 5 14>, angularLimitsMin=<-50 -5 -60>, angularTarget=<0 0 0>}
      localJointOffset=<-5 -1 -5>
     }]
    relativeSpaceBone {boneName=root}
   }
   n7 = Get offset_weapon_0() -> (offset_weapon_0:double)
   n8 = Anim Dynamics - Bone: dyn_scarf_r(ComponentPose=@n6, Alpha=@n7, ExternalForce="0,0,0") {
    angularSpringConstant=800
    bAngularSpring=true
    bGravityOverrideInSimSpace=true
    bOverrideAngularDamping=true
    bUseGravityOverride=true
    gravityScale=0.3
    simulationSpace=BoneRelative
    alphaBoolBlend {blendOption=Linear}
    boundBone {boneName=dyn_scarf_r}
    componentAppliedLinearAccClamp=<400 400 400>
    componentLinearAccScale=<0.1 0.1 0.1>
    componentLinearVelScale=<0.1 0.1 0.1>
    gravityOverride {z=-980}
    physicsBodyDefinitions [{
      sphereCollisionRadius=0
      boundBone {boneName=dyn_scarf_r}
      boxExtents=<45 45 45>
      constraintSetup {twistAxis=AxisZ, angularLimitsMax=<0 5 14>, angularLimitsMin=<-50 -5 -60>, angularTarget=<0 0 0>}
      localJointOffset=<-5 -1 -5>
     }]
    relativeSpaceBone {boneName=root}
   }
   n9 = Anim Dynamics - Bone: dyn_scarf_l_01(ComponentPose=@n8, Alpha=0.4, ExternalForce="0,0,0") {
    bAngularSpring=true
    bGravityOverrideInSimSpace=true
    bOverrideAngularDamping=true
    bUseGravityOverride=true
    gravityScale=0.3
    simulationSpace=BoneRelative
    alphaBoolBlend {blendOption=Linear}
    boundBone {boneName=dyn_scarf_l_01}
    componentAppliedLinearAccClamp=<400 400 400>
    componentLinearAccScale=<0.1 0.1 0.1>
    componentLinearVelScale=<0.1 0.1 0.1>
    gravityOverride {z=-980}
    physicsBodyDefinitions [{
      sphereCollisionRadius=0
      boundBone {boneName=dyn_scarf_l_01}
      boxExtents=<30 30 30>
      constraintSetup {angularLimitsMax=<0 5 14>, angularLimitsMin=<-50 -5 -60>, angularTarget=<0 0 0>}
      localJointOffset=<-5 -0.1 0.1>
     }]
    relativeSpaceBone {boneName=root}
   }
   n10 = Anim Dynamics - Bone: dyn_scarf_r_01(ComponentPose=@n9, Alpha=@n7, ExternalForce="0,0,0") {
    bAngularSpring=true
    bGravityOverrideInSimSpace=true
    bOverrideAngularDamping=true
    bUseGravityOverride=true
    gravityScale=0.3
    simulationSpace=BoneRelative
    alphaBoolBlend {blendOption=Linear}
    boundBone {boneName=dyn_scarf_r_01}
    componentAppliedLinearAccClamp=<400 400 400>
    componentLinearAccScale=<0.1 0.1 0.1>
    componentLinearVelScale=<0.1 0.1 0.1>
    gravityOverride {z=-980}
    physicsBodyDefinitions [{
      sphereCollisionRadius=0
      boundBone {boneName=dyn_scarf_r_01}
      boxExtents=<30 30 30>
      constraintSetup {angularLimitsMax=<0 5 14>, angularLimitsMin=<-50 -5 -60>, angularTarget=<0 0 0>}
      localJointOffset=<-5 -0.1 0.1>
     }]
    relativeSpaceBone {boneName=root}
   }
   Output Pose(Result=Component To Local(ComponentPose=@n10)) {layerGroup=None}
  ]}
 EventGraph {nodes [
   Event BlueprintUpdateAnimation()
   > Set offset_weapon(offset_weapon:double=0) -> (Output_Get:double)
   > Set offset_weapon_0(offset_weapon_0:double=0.4) -> (Output_Get:double)
   > n1 = Switch on EFortWeaponCoreAnimation!(execute, Selection:EFortWeaponCoreAnimation Enum=Get WeaponCoreAnim!() -> (WeaponCoreAnim)) -> (
    Melee:Exec
    Pistol:Exec
    Shotgun:Exec
    PaperBlueprint:Exec
    Rifle:Exec
    MeleeOneHand:Exec
    MachinePistol:Exec
    RocketLauncher:Exec
    GrenadeLauncher:Exec
    GoingCommando:Exec
    AssaultRifle:Exec
    TacticalShotgun:Exec
    SniperRifle:Exec
    TrapPlacement:Exec
    ShoulderLauncher:Exec
    AbilityDecoTool:Exec
    Crossbow:Exec
    C4:Exec
    RemoteControl:Exec
    DualWield:Exec
    AR_BullPup:Exec
    AR_ForwardGrip:Exec
    MedPackPaddles:Exec
    SMG_P90:Exec
    AR_DrumGun:Exec
    Consumable_Small:Exec
    Consumable_Large:Exec
    Balloon:Exec
    MountedTurret:Exec
    CreativeTool:Exec
    ExplosiveBow:Exec
    AshtonIndigo:Exec
    AshtonChicago:Exec
    MeleeDualWield:Exec
    Unarmed:Exec
   )
   n2 = Branch(execute=[@n3.else, @n1.AssaultRifle, @n1.Shotgun, @n1.TacticalShotgun], Condition=Get bIsTargeting())
   > Set offset_weapon(offset_weapon=1)
   > Set offset_weapon_0(offset_weapon_0=0.2)
   Set offset_weapon(execute=@n2.else, offset_weapon=1)
   > Set offset_weapon_0(offset_weapon_0=0.4)
   n3 = Branch(execute=@n1.Pistol, Condition=Get bIsPlayingEmote())
   > Set offset_weapon(offset_weapon=0)
  ]}
]
variables [
 offset_weapon {type=number}
 offset_weapon_0 {type=number}
]
